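# Conversation database for the Aura NPC, providing the dialogue tree entered via the Conversation.Entry.Aura tag.
# A single Speech node shows a default chalice-lore line, swapped by prioritized contextual messages for first-ever conversations or reactions to the player's specific outfit.
# The speech then branches through link nodes into shared basic services, NPC library quests, and three Season 25 quest-specific Aura conversations.
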ConversationDatabase /NPCLibrary/NPCs/Aura/NPCLibrary_Conversation_Aura
entry_points [{tag=Conversation.Entry.Aura, destinations [#1]}]
nodes [
 ConversationEntryPointNode_7 {
  id=#1
  class=ConversationEntryPointNode
  outputs [#2]
  properties {entryTag {tagName=Conversation.Entry.Aura}}
 }
 Speech_C_0 {
  id=#2
  class=Speech_C
  outputs [
   #3
   #4
   #5
   #6
   #7
  ]
  properties {
   nodeName=""
   generalConfig {
    defaultMessage=They say someone altered the chalice... Gave it power.
    contextualMessages [
     {
      message=Someone has to find the Chalice of Cubetastrophe before it's too late.
      contextRequirements [{requirement {class=FortControllerRequirement_IsFirstEverConversationWithNPC}}]
      priority {curve {rowName=Tandem.Speech.Priority.FirstEverInteraction}}
     }
     {
      message=I'm trying to do the right thing.
      contextRequirements [{requirement {class=FortControllerRequirement_HasCID, properties {allowedCIds [/Game/Athena/Items/Cosmetics/Characters/CID_206_Athena_Commando_M_Bling]}}}]
      priority {curve {rowName=Tandem.Speech.Priority.OutfitReaction}}
     }
     {
      message=There's something in this rainforest that doesn't belong.
      contextRequirements [{requirement {class=FortControllerRequirement_HasCID, properties {allowedCIds [/BRCosmetics/Athena/Items/Cosmetics/Characters/Character_ZirconSweep]}}}]
      priority {curve {rowName=Tandem.Speech.Priority.OutfitReaction}}
     }
     {
      message=I've never met anyone more reckless than you.
      contextRequirements [{requirement {class=FortControllerRequirement_HasCID, properties {allowedCIds [/BRCosmetics/Athena/Items/Cosmetics/Characters/Character_CrispRover]}}}]
      priority {curve {rowName=Tandem.Speech.Priority.OutfitReaction}}
     }
     {
      message=I don't need your help. I'll find it myself.
      contextRequirements [{requirement {class=FortControllerRequirement_HasCID, properties {allowedCIds [/BRCosmetics/Athena/Items/Cosmetics/Characters/Character_RippedHarvester]}}}]
      priority {curve {rowName=Tandem.Speech.Priority.OutfitReaction}}
     }
     {
      message=For once... I just wanna do something good.
      contextRequirements [{requirement {
         class=FortControllerRequirement_HasCID
         properties {allowedCIds [/BRCosmetics/Athena/Items/Cosmetics/Characters/CID_398_Athena_Commando_M_TreasureHunterFashion]}
        }}]
      priority {curve {rowName=Tandem.Speech.Priority.OutfitReaction}}
     }
     {
      message=I know you brought the Chalice of Cubetastrophe through the Rift Gate. I just can't prove it.
      contextRequirements [{requirement {class=FortControllerRequirement_HasCID, properties {allowedCIds [/BRCosmetics/Athena/Items/Cosmetics/Characters/Character_Lilac]}}}]
      priority {curve {rowName=Tandem.Speech.Priority.OutfitReaction}}
     }
    ]
   }
  }
 }
 ConversationLinkNode_4 {
  id=#7
  class=ConversationLinkNode
  properties {remoteEntryTag {tagName=Conversation.Entry.Season25.P01.Q01.Aura}}
 }
 ConversationLinkNode_3 {
  id=#6
  class=ConversationLinkNode
  properties {remoteEntryTag {tagName=Conversation.Entry.Season25.P01.Q04.Aura}}
 }
 ConversationLinkNode_2 {
  id=#5
  class=ConversationLinkNode
  properties {remoteEntryTag {tagName=Conversation.Entry.Season25.P01.Q03.Aura}}
 }
 ConversationLinkNode_0 {id=#4, class=ConversationLinkNode, properties {remoteEntryTag {tagName=Conversation.Entry.BasicServices}}}
 ConversationLinkNode_1 {id=#3, class=ConversationLinkNode, properties {remoteEntryTag {tagName=Conversation.Entry.NPCLibrary.Quests}}}
]
